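# Translucent UI material for the Thermal Lance overheat indicator, drawing a radial progress gauge with procedural frame shapes.
# The Progress parameter fills an arc that is masked to a thin circle outline.
# SDF circles and boxes, repeated through rotated radial-mirror UVs, form an outer frame and spiked frame whose segment count follows PulseNumber.
# Named reroutes carry masks and colors between sections, with BarOpacity and MasterOpacity scaling transparency.
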Material /ThermalLanceGameplay/UI/Materials/M_UI_ThermalLance_Overheat
graph {
 properties {blendMode=BLEND_Translucent, materialDomain=MD_UI, pixelDepthOffsetMode=PDOM_Legacy, usedShadingModels=Default Lit}
 nodes [
  Cosine(None=Multiply(A=Time(), B=Constant() {r=0.162}) {constB=0.648})
  MaterialFunctionCall(UVs=NamedRerouteUsage(), "Rotation Angle (0-1)"=Constant() {r=0.5}) {materialFunction=&CustomRotator}
  n1 = NamedRerouteUsage()
  n2 = ScalarParameter() {defaultValue=0.85, parameterName=BarOpacity}
  Multiply(
   A=LinearInterpolate(A=NamedRerouteUsage(), B=@n1, Alpha=Multiply(A=@n1, B=@n2))
   B=ScalarParameter() {defaultValue=1, parameterName=MasterOpacity}
  )
  NamedRerouteDeclaration(None=MaterialFunctionCall() {materialFunction=&GetUserInterfaceUV}.Normalized UV) {name=UV, nodeColor=<rgba 0 0 0 1>}
  n3 = NamedRerouteUsage()
  n4 = ScalarParameter() {defaultValue=6, parameterName=PulseNumber}
  n5 = Multiply(A=@n4) {constB=4}
  n6 = Divide(B=@n5) {constA=0.5}
  n7 = Divide(A=ScalarParameter() {defaultValue=0.95, parameterName=FinalPulseTime}, B=@n4) {constA=0.95}
  n8 = MaterialFunctionCall!(
   UVs=MaterialFunctionCall(UVs=TextureCoordinate(), "Rotation Angle (0-1)"=Add(A=@n6, B=@n7) {constB=0.1583}) {materialFunction=&CustomRotator}.Rotated Values
   Sides=@n5
  ) {materialFunction=/FortUILibrary/MaterialFunctions/Basics/MF_UI_UVRadialMirror}
  n9 = ComponentMask(None=@n8.UVs) {r=true}
  n10 = Add(A=ComponentMask(None=@n8.UVs) {g=true}, B=Constant() {r=0.18})
  n11 = NamedRerouteUsage()
  n12 = Constant()
  n13 = Max(
   A=MaterialFunctionCall!(UVs=AppendVector(A=@n9, B=@n10), Size=Constant2Vector() {g=0.1, r=0.01}, CornerRoundness, KeepAspectRatio, PixelSize) {materialFunction=/FortUILibrary/MaterialFunctions/Shapes/SDFs/MF_SDF_Box}.SDF
   B=Multiply(A=Add(A=@n11, B=@n12)) {constB=-1}
  )
  n14 = Multiply(A=Min(A=Saturate(None=@n3), B=Saturate(None=@n13))) {constB=2}
  n15 = Multiply(A=MaterialFunctionCall(DistanceField=@n14, Threshold=Constant() {r=0.01}, Smoothing=Constant() {r=0.1}) {materialFunction=&MF_UI_AntiAliasing}.Inner) {constB=2}
  n16 = MaterialFunctionCall(DistanceField=@n14, Threshold=Constant() {r=0.03}, Smoothing=Constant() {r=0.1}) {materialFunction=&MF_UI_AntiAliasing}
  NamedRerouteDeclaration(None=Multiply(A=LinearInterpolate(A=@n15, B=@n16.Inner) {constAlpha=0.6}) {constB=1.25}) {name=SpikedShadowFrame, nodeColor=<rgba 1 0.375 0 1>}
  n17 = MaterialFunctionCall(
   Size=Constant() {r=0.625}
   OutlineThickness=ScalarParameter() {defaultValue=0.018, group=AmmoCounter, parameterName=RadialThickness, sliderMax=1}
  ) {materialFunction=&MF_UI_Circle}
  NamedRerouteDeclaration(None=OneMinus(None=SmoothStep(Value=@n17.SDFOutline) {constMax=0.01})) {name=Mask_Radial, nodeColor=<rgba 0 0 0 1>}
  n18 = NamedRerouteUsage()
  n19 = OneMinus(None=NamedRerouteUsage())
  n20 = MaterialFunctionCall!(
   UV
   "Arc Start Angle (0-1)"=ScalarParameter() {defaultValue=0.25, parameterName=Start Angle, sliderMax=1, sortPriority=31}
   "Arc End Angle (0-1)"=ScalarParameter() {defaultValue=0.5, parameterName=End Angle, sliderMax=1}
   "Fill Ratio (0-1)"=NamedRerouteUsage()
   Fill Clockwise=Constant()
  ) {materialFunction=/FortUILibrary/MaterialFunctions/Basics/MF_UI_RadialArc}
  NamedRerouteDeclaration(None=Multiply(A=Multiply(A=@n18, B=@n19), B=SmoothStep(Value=@n20.Radial Fill) {constMax=0.001})) {name=Mask_OverheatProgress, nodeColor=<rgba 0 0 0 1>}
  n21 = TextureCoordinate()
  n22 = Constant2Vector() {g=-0.12, r=-0.146}
  NamedRerouteDeclaration(None=MaterialFunctionCall(UVs=Add(A=@n21, B=@n22), Texture Scale=Constant() {r=0.36}) {materialFunction=&ScaleUVsByCenter}.UVs) {name=TextureUVs, nodeColor=<rgba 0 0 0 1>}
  n23 = TextureCoordinate()
  n24 = MaterialFunctionCall(UVs=ConstantBiasScale(None=@n23) {bias=-0.5, scale=1}, Size=Constant() {r=1}) {materialFunction=&MF_SDF_Box}
  NamedRerouteDeclaration(None=Multiply(A=@n24.SDF) {constB=-1}) {name=BLQuad2, nodeColor=<rgba 0 0 0 1>}
  n25 = Constant3Vector()
  n26 = VectorParameter() {parameterName=FrameColor, defaultValue=<rgba 1 1 1 1>}
  n27 = NamedRerouteUsage()
  NamedRerouteDeclaration(None=LinearInterpolate(A=NamedRerouteUsage(), B=LinearInterpolate(A=@n25, B=@n26.RGB, Alpha=@n27), Alpha=NamedRerouteUsage())) {name=Outer Frame Color, nodeColor=<rgba 0 0 0 1>}
  n28 = MaterialFunctionCall!(UVs, Radius=Constant() {r=0.7}) {materialFunction=/FortUILibrary/MaterialFunctions/Shapes/SDFs/MF_SDF_Circle}
  n29 = Abs(None=ConstantBiasScale(None=@n23) {bias=-0.5, scale=2})
  n30 = Min(A=NamedRerouteUsage(), B=Min(A=ComponentMask(None=@n29) {g=true}, B=ComponentMask(None=@n29) {r=true}))
  n31 = MaterialFunctionCall(UVs=Rotator(Coordinate=@n23, Time=Constant() {r=0.5}) {speed=3.1416}, Sides=Constant() {r=4}) {materialFunction=&MF_UI_UVRadialMirror}
  n32 = Constant() {r=0.15}
  n33 = Rotator(Coordinate=@n31.UVs, Time=@n32) {speed=3.1416}
  n34 = Constant() {r=-0.085}
  n35 = Rotator(Coordinate=@n31.UVs, Time=OneMinus(None=@n32)) {speed=3.1416}
  n36 = Min(A=ComponentMask(None=Add(A=@n33, B=@n34) {constB=-0.1}) {r=true}, B=ComponentMask(None=Add(A=@n35, B=@n34) {constB=-0.1}) {r=true})
  n37 = Constant() {r=0.25}
  n38 = ComponentMask(None=Add(A=Rotator(Coordinate=@n23, Time=@n37) {speed=3.1416}) {constB=-0.65}) {g=true}
  n39 = Constant() {r=0.02}
  n40 = Min(
   A=Max(A=Max(A=@n30, B=@n36), B=Multiply(A=@n38) {constB=-2})
   B=Max(
    A=MaterialFunctionCall!(DistanceField=@n28.SDF, Threshold, Thickness=@n39, Alignment, Smoothing) {materialFunction=/FortUILibrary/MaterialFunctions/Shapes/Modifiers/MF_UI_ShapeOutline}.SDFOutline
    B=@n24.SDF
   )
  )
  NamedRerouteDeclaration(None=Multiply(A=Max(A=@n28.SDF, B=@n40)) {constB=3}) {name=OuterFrameSDF, nodeColor=<rgba 0 0 0 1>}
  NamedRerouteDeclaration(None=VectorParameter() {group=Color, parameterName=OverheatColorB, defaultValue=<rgba 0.9047 0.7231 0.0343 1>}.RGB) {name=Overheat Color, nodeColor=<rgba 0 0 0 1>}
  NamedRerouteDeclaration(None=@n15) {name=SpikedFrame, nodeColor=<rgba 0.6172 1 0 1>}
  NamedRerouteDeclaration(None=ScalarParameter() {group=Color, parameterName=Progress, sliderMax=1}) {name=Progress, nodeColor=<rgba 0 0 0 1>}
  NamedRerouteUsage()
  NamedRerouteUsage()
 ]
}
