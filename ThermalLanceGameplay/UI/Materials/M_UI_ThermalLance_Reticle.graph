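# UI material that draws the Thermal Lance reticle procedurally from SDF shapes, reacting to target range.
# The inner circle is cut by rotatable, radially mirrored box gaps that close with EnvInRange, and goes from ring to filled dot with EnemyInRange.
# Outer brackets are masked from a box-circle union outline, whose box widens with EnvInRange.
# Offset antialiasing thresholds add drop shadows, and EnemyInRange tints the reticle from white to yellow.
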Material /ThermalLanceGameplay/UI/Materials/M_UI_ThermalLance_Reticle
graph {
 properties {bAutomaticallySetUsageInEditor=false, blendMode=BLEND_Translucent, materialDomain=MD_UI, usedShadingModels=Default Lit}
 nodes [
  n1 = Constant3Vector() {constant=<rgba 0.9047 0.7231 0.0343 1>}
  n2 = ScalarParameter() {parameterName=EnemyInRange}
  n3 = LinearInterpolate(B=@n1, Alpha=@n2) {constA=1}
  n4 = NamedRerouteUsage()
  LinearInterpolate(
   A=Add(A=LinearInterpolate(B=@n3, Alpha=@n4), B=NamedRerouteUsage())
   B=LinearInterpolate(B=@n1, Alpha=@n2) {constA=1}
   Alpha=NamedRerouteUsage()
  )
  n5 = NamedRerouteUsage()
  n6 = ScalarParameter() {defaultValue=0.85, parameterName=OuterShapeOpacity}
  Multiply(A=Add(A=Multiply(A=@n5, B=@n6) {constB=0.45}, B=NamedRerouteUsage())) {constB=0.9}
  n7 = ScalarParameter() {parameterName=EnemyInRange}
  n8 = MaterialFunctionCall!(UVs=TextureCoordinate(), Radius=LinearInterpolate(Alpha=@n7) {constA=0.04, constB=0.04}) {materialFunction=/FortUILibrary/MaterialFunctions/Shapes/SDFs/MF_SDF_Circle}
  n9 = ScalarParameter() {defaultValue=0.0175, parameterName=CircleThickness}
  n10 = Constant() {r=1}
  n11 = ScalarParameter() {parameterName=Rotation}
  n12 = MaterialFunctionCall!(UVs=Rotator(Coordinate=TextureCoordinate(), Time=LinearInterpolate(Alpha=@n11) {constB=0.25}) {speed=3.1416}, Sides=Constant() {r=4}) {materialFunction=/FortUILibrary/MaterialFunctions/Basics/MF_UI_UVRadialMirror}
  n13 = AppendVector(
   A=ComponentMask(None=@n12.UVs) {r=true}
   B=Add(A=ComponentMask(None=@n12.UVs) {g=true}, B=ScalarParameter() {defaultValue=-0.5, parameterName=DashOffset})
  )
  n14 = ScalarParameter() {parameterName=EnvInRange}
  n15 = AppendVector(A=LinearInterpolate(Alpha=@n14) {constA=0.0325, constB=-0.1}, B=OneMinus(None=@n14))
  n16 = MaterialFunctionCall!(
   DistanceFieldA=LinearInterpolate(
    A=MaterialFunctionCall!(DistanceField=@n8.SDF, Threshold, Thickness=@n9, Alignment=@n10, Smoothing) {materialFunction=/FortUILibrary/MaterialFunctions/Shapes/Modifiers/MF_UI_ShapeOutline}.SDFOutline
    B=@n8.SDF
    Alpha=@n7
   )
   DistanceFieldB=Multiply(A=MaterialFunctionCall!(UVs=@n13, Size=@n15, CornerRoundness, KeepAspectRatio, PixelSize) {materialFunction=/FortUILibrary/MaterialFunctions/Shapes/SDFs/MF_SDF_Box}.SDF) {constB=0.45}
   BlendAmount
  ) {materialFunction=/FortUILibrary/MaterialFunctions/Shapes/Modifiers/MF_UI_SDF_Subtraction}
  n17 = NamedRerouteDeclaration(None=MaterialFunctionCall(DistanceField=@n16.Exact, Threshold=Constant()) {materialFunction=&MF_UI_AntiAliasing}.Inner) {name=InnerCircle, nodeColor=<rgba 1 1 1 1>}
  n18 = MaterialFunctionCall(DistanceField=@n16.Exact, Threshold=Constant() {r=0.0113}) {materialFunction=&MF_UI_AntiAliasing}
  NamedRerouteDeclaration(None=Multiply(A=LinearInterpolate(A=@n17, B=@n18.Inner) {constAlpha=0.6}, B=ScalarParameter() {defaultValue=0.85, parameterName=InnerCircle Opacity})) {name=InnerCircle+Shadow, nodeColor=<rgba 1 1 1 1>}
  n19 = ScalarParameter() {parameterName=EnvInRange}
  n20 = MaterialFunctionCall(Size=AppendVector(A=LinearInterpolate(Alpha=@n19) {constA=0.55, constB=0.7}, B=Constant() {r=0.2})) {materialFunction=&MF_SDF_Box}
  n21 = TextureCoordinate()
  n22 = MaterialFunctionCall(UVs=@n21, Radius=Constant() {r=0.55}) {materialFunction=&MF_SDF_Circle}
  n23 = Subtract(A=@n21) {constB=0.5}
  n24 = Abs(None=@n23)
  n25 = ComponentMask(None=@n24) {g=true}
  n26 = Max(
   A=MaterialFunctionCall(
    DistanceField=MaterialFunctionCall!(DistanceFieldA=@n20.SDF, DistanceFieldB=@n22.SDF, BlendAmount) {materialFunction=/FortUILibrary/MaterialFunctions/Shapes/Modifiers/MF_UI_SDF_Union}.Exact
    Thickness=ScalarParameter() {defaultValue=0.025, parameterName=OuterThickness}
    Alignment=Constant() {r=-1}
   ) {materialFunction=&MF_UI_ShapeOutline}.SDFOutline
   B=Multiply(A=Subtract(A=@n25) {constB=0.075, desc=Middle Thickness}) {constB=-2}
  )
  n27 = ComponentMask(None=@n23) {g=true}
  n28 = Subtract(A=ComponentMask(None=@n24) {r=true}) {constB=0.3125, desc=Bottom Pip Length}
  n29 = Max(A=@n26, B=Multiply(A=Min(A=@n27, B=@n28)) {constB=2})
  n30 = MaterialFunctionCall(DistanceField=@n29) {materialFunction=&MF_UI_AntiAliasing}
  n31 = MaterialFunctionCall(DistanceField=@n29, Threshold=Constant() {r=0.0113}) {materialFunction=&MF_UI_AntiAliasing}
  NamedRerouteDeclaration(None=LinearInterpolate(
    A=NamedRerouteDeclaration(None=@n30.Inner) {name=Outer Shape, nodeColor=<rgba 1 1 1 1>}
    B=NamedRerouteDeclaration(None=@n31.Inner) {name=Outer Shape Dropshadow, nodeColor=<rgba 1 1 1 1>}
   ) {constAlpha=0.4}) {name=Shape+Shadow, nodeColor=<rgba 1 1 1 1>}
  n32 = Max(
   A=MaterialFunctionCall(DistanceField=@n22.SDF, Thickness=Constant() {r=0.02}, Alignment=Constant() {r=-2}) {materialFunction=&MF_UI_ShapeOutline}.SDFOutline
   B=@n26
  )
  NamedRerouteDeclaration(None=MaterialFunctionCall(DistanceField=Max(A=@n32, B=@n29)) {materialFunction=&MF_UI_AntiAliasing}.Inner) {name=OuterColorBracket, nodeColor=<rgba 1 1 1 1>}
 ]
}
