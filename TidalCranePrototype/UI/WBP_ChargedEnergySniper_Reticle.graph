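# Reticle extension widget for the Charged Energy Sniper, built on Thermal Lance materials, with a center reticle and an overheat meter.
# Construct binds weapon overheat and targeting-component events, and sets meter pulse parameters from a curve-table cadence loop.
# Tick also feeds weapon charge into the overheat handler, so the meter doubles as a charge gauge that animates and timer-hides.
# Delta-gated functions animate the center material for environment or enemy targets.
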WidgetBlueprint /TidalCranePrototype/UI/WBP_ChargedEnergySniper_Reticle
animations [
 Anim_OnOverheat {
  binding_count=1
  length_seconds=0.35
  object=&self:Anim_OnOverheat
  sequence {bindings [Image_OverheatMeter {
     id=#1
     possessed_class=Image
     tracks [{
       type=MovieSceneFloatTrack
       display_name=Render Opacity
       sections [{type=MovieSceneFloatSection, end_frame=7, start_frame=0, channels [None {type=MovieSceneScriptingFloatChannel, keys [[0,1],[7,0.2]]}]}]
      }]
    }]}
 }
 Anim_OverheatStart {
  binding_count=1
  length_seconds=0.35
  object=&self:Anim_OverheatStart
  sequence {bindings [Image_OverheatMeter {
     id=#2
     possessed_class=Image
     tracks [{
       type=MovieSceneFloatTrack
       display_name=Render Opacity
       sections [{
         type=MovieSceneFloatSection
         end_frame=7
         start_frame=0
         channels [None {type=MovieSceneScriptingFloatChannel, keys [[-8.0003,1],[0,0],[7,0.85],[7.0003,1]]}]
        }]
      }]
    }]}
 }
 OnInRange {
  binding_count=1
  length_seconds=0.0667
  object=&self:OnInRange
  sequence {
   bindings [Image_ReticleCenter {
     id=#3
     possessed_class=Image
     tracks [{
       type=MovieSceneWidgetMaterialTrack
       sections [{
         type=MovieSceneParameterSection
         end_frame=2
         start_frame=0
         channels [
          DashWidth {type=MovieSceneScriptingFloatChannel, keys [[0,0],[1,1.1]]}
          InnerCircle Opacity {type=MovieSceneScriptingFloatChannel, keys [[0,0.7],[1,0.95]]}
          OuterShapeOpacity {type=MovieSceneScriptingFloatChannel, keys [[0,0.5],[1,0.95]]}
          CircleThickness {type=MovieSceneScriptingFloatChannel, const [0.015,3]}
         ]
        }]
      }]
    }]
   display_rate {numerator=30}
  }
 }
 Anim_Env_in_Range {
  binding_count=1
  length_seconds=0.1
  object=&self:Anim_Env_in_Range
  sequence {
   bindings [Image_ReticleCenter {
     id=#4
     possessed_class=Image
     tracks [{
       type=MovieSceneWidgetMaterialTrack
       sections [{
         type=MovieSceneParameterSection
         end_frame=3
         start_frame=0
         channels [EnvInRange {type=MovieSceneScriptingFloatChannel, keys [[0,0],[3,1]]}]
        }]
      }]
    }]
   display_rate {numerator=30}
  }
 }
 Anim_Enemy_in_Range {
  binding_count=1
  length_seconds=0.1
  object=&self:Anim_Enemy_in_Range
  sequence {
   bindings [Image_ReticleCenter {
     id=#5
     possessed_class=Image
     tracks [{
       type=MovieSceneWidgetMaterialTrack
       sections [{
         type=MovieSceneParameterSection
         end_frame=3
         start_frame=0
         channels [EnemyInRange {type=MovieSceneScriptingFloatChannel, keys [[0,0],[3,1]]}]
        }]
      }]
    }]
   display_rate {numerator=30}
  }
 }
]
graphs [
 CalculatePulseProperties {nodes [
   CalculatePulseProperties() {is=title K2Node_FunctionEntry}
   n1 = GetValueAtLevel!(
    Input:"Scalable Float Structure (by ref)"=Get Hotfix_PulseFireOverheatCadence() -> (Hotfix_PulseFireOverheatCadence:Scalable Float Structure)
    Level:float="To Float (Integer)"(InInt=Get Index() -> (Index:int=0))
   ) -> (ReturnValue:float)
   > n2 = While Loop!(execute:Exec, Condition:bool="float >= float"(A=@n1)) -> (LoopBody:Exec, Completed:Exec) {is=title K2Node_MacroInstance}
   Return Node!(
    execute=@n2.Completed
    Count:int=Get Index() -> (Index=0)
    FinalPulseValue:double=Get FinalPulseValueReturn() -> (FinalPulseValueReturn:double)
   )
   > Set FinalPulseValueReturn(FinalPulseValueReturn:double=@n1) -> (Output_Get:double)
   > Increment Int(Value=Get Index() -> (Index=0))
  ]}
 EventGraph {nodes [
   n1 = HandleOverheat!() -> (OutputDelegate, then, CurrentValue:float=~, MaxValue:float=~, State:EFortWeaponOverheatState Enum=None, bIsInCooldown:bool=~) {is=title K2Node_CustomEvent}
   > n2 = Set CurrentOverheatValue(CurrentOverheatValue:float=@n1.CurrentValue) -> (Output_Get:float)
   > SetScalarParameterValue(
    self=Get DynamicMaterialOverheat() -> (DynamicMaterialOverheat:Material Instance Dynamic Object Reference)
    ParameterName=Progress
    Value=float - float(A=1, B=float / float(A=@n2, B=@n1.MaxValue))
   )
   HideOverheatWithoutPlayerUsage() {is=title K2Node_CustomEvent}
   n3 = Get bOverheatHidden() -> (bOverheatHidden:bool)
   n4 = Get CurrentOverheatValue() -> (CurrentOverheatValue:float)
   > Branch(Condition=AND Boolean(A=NOT Boolean(A=@n3), B="Equal (Float)"(A=@n4)))
   > PlayAnimation(InAnimation=Get Anim_OverheatStart() -> (Anim_OverheatStart:Widget Animation Object Reference), PlayMode=Reverse)
   > Set bOverheatHidden(bOverheatHidden:bool=true) -> (Output_Get:bool)
   n5 = HandleTargetedActorTypeChanged!() -> (OutputDelegate, then, ActorType:AbilityTargetingTargetActorTypeEnum Enum=NewEnumerator0) {is=title K2Node_CustomEvent}
   > n6 = Switch on AbilityTargetingTargetActorTypeEnum!(execute, Selection:AbilityTargetingTargetActorTypeEnum Enum=@n5.ActorType) -> (NewEnumerator0:Exec, NewEnumerator1:Exec, NewEnumerator2:Exec) {is=title K2Node_SwitchEnum}
   > PlayActorInRange!(execute, self, IsActorInRange?:bool=false) -> (then)
   > PlayEnemyInRange!(execute, self, IsEnemyInRange:bool=false) -> (then)
   PlayActorInRange(execute=@n6.NewEnumerator1, IsActorInRange?=true)
   > PlayEnemyInRange(IsEnemyInRange=false)
   PlayEnemyInRange(execute=@n6.NewEnumerator2, IsEnemyInRange=true)
   > PlayActorInRange(IsActorInRange?=true)
   Event Tick()
   > n7 = Get!(execute:Exec) -> (then:Exec, else:Exec, Weapon:Fort Weapon Ranged Object Reference)
   > Branch(Condition="float > float"(A=GetChargeTime!(self:Fort Weapon Object Reference=@n7) -> (ReturnValue:float)))
   n8 = GetChargeToMinPercent!(self:Fort Weapon Object Reference=@n7) -> (ReturnValue:float)
   n9 = float / float(A=@n8, B=3)
   n10 = "float >= float"(A=@n8, B=1)
   n11 = float + float(A=GetChargePercent!(self:Fort Weapon Object Reference=@n7) -> (ReturnValue:float), B=0.33)
   > HandleOverheat!(
    execute
    self
    CurrentValue:float="Clamp (Float)"(Value=SelectFloat(A=@n11, B=@n9, bPickA=@n10), Min=0, Max=1)
    MaxValue:float=1
    State:EFortWeaponOverheatState Enum=Heating
    bIsInCooldown:bool=false
   ) -> (then) {is=title K2Node_CallFunction}
   Event Destruct()
   > n12 = Sequence()
   > n13 = Get()
   > Unbind Event from On Overheat Value Changed(self:Fort Weapon Ranged Object Reference=@n13, Delegate=Create Event()) {is=title K2Node_RemoveDelegate}
   n14 = GetComponentByClass!(
    self:Actor Object Reference=@n13
    ComponentClass:Actor Component Class Reference=/ThermalLanceGameplay/Gameplay/Targeting/B_ThermalLanceWeaponTargetingComponent.B_ThermalLanceWeaponTargetingComponent_C
   ) -> (ReturnValue:B Thermal Lance Weapon Targeting Component Object Reference)
   > Is Valid(InputObject=@n14)
   > Unbind Event from On Targeted Actor Type Changed(self:B Ability Targeting Feedback Component Parent Object Reference=@n14, Delegate=Create Event()) {is=title K2Node_RemoveDelegate}
   Parent: Destruct(execute=@n12.then_1)
   Event Construct()
   > Parent: Construct()
   > SetupOverheatMeter()
   > n15 = GetDynamicMaterial(self=Get Image_ReticleCenter() -> (Image_ReticleCenter:Image Object Reference))
   > n16 = Set M_CenterReticle(M_CenterReticle:Material Instance Dynamic Object Reference=@n15) -> (Output_Get:Material Instance Dynamic Object Reference)
   > SetScalarParameterValue(self=@n16, ParameterName=OuterShapeOpacity, Value=0.375)
   > n17 = Cast To FortWeaponRanged!(
    execute
    Object=GetEquippedWeapon!(self:Fort Weapon Reticle Extension Widget Base Object Reference) -> (ReturnValue:Fort Weapon Object Reference)
   ) -> (then, CastFailed, AsFort Weapon Ranged:Fort Weapon Ranged Object Reference) {is=title K2Node_DynamicCast}
   > n18 = Set Weapon(Weapon:Fort Weapon Ranged Object Reference=@n17) -> (Output_Get:Fort Weapon Ranged Object Reference)
   PlayAnimation(execute=[@n26.then, @n18.then], InAnimation=Get Anim_OverheatStart())
   > Set bOverheatHidden(bOverheatHidden=false)
   > HandleOverheat(CurrentValue=0, MaxValue=1, State=Heating, bIsInCooldown=false)
   n19 = Branch(Condition="Equal (Float)"(A=Get CurrentOverheatValue(), B=@n1.MaxValue))
   > Set bWasOverheated(bWasOverheated:bool=true) -> (Output_Get:bool)
   > PlayAnimation(InAnimation=Get Anim_OnOverheat() -> (Anim_OnOverheat:Widget Animation Object Reference))
   n20 = Branch(execute=@n19.else, Condition="Equal (Float)"(A=Get CurrentOverheatValue()))
   > n21 = Branch(Condition=Get bWasOverheated() -> (bWasOverheated:bool))
   > Set bWasOverheated(bWasOverheated=false)
   > n22 = "Play Animation Time Range with Finished event (legacy)"!(
    execute:Exec
    Widget:User Widget Object Reference=Self-Reference()
    InAnimation:Widget Animation Object Reference=Get Anim_OnOverheat()
    StartAtTime:float=0
    EndAtTime:float=0
    NumLoopsToPlay:int=1
    PlayMode:EUMGSequencePlayMode Enum=Reverse
    PlaybackSpeed:float=1
   ) -> (then:Exec, Result:UMGSequence Player Object Reference, Finished:Exec) {is=title K2Node_PlayAnimationTimeRange}
   n23 = Set Timer by Event(
    execute=[@n22.Finished, @n21.else]
    Delegate=Create Event()
    Time=Get TimeToHideOverheat() -> (TimeToHideOverheat:float)
    bLooping=false
    bMaxOncePerFrame=true
   )
   > Set HideOverheatDelayTimerHandle(HideOverheatDelayTimerHandle:Timer Handle Structure=@n23) -> (Output_Get:Timer Handle Structure)
   Branch(execute=@n20.else, Condition="float > float"(A=Get CurrentOverheatValue()))
   > n24 = Branch(Condition=Get bOverheatHidden())
   > PlayAnimation(InAnimation=Get Anim_OverheatStart())
   > Set bOverheatHidden(bOverheatHidden=false)
   Branch(
    execute=@n24.else
    Condition=Is Valid Timer Handle!(Handle:Timer Handle Structure=Get HideOverheatDelayTimerHandle() -> (HideOverheatDelayTimerHandle:Timer Handle Structure)) -> (ReturnValue:bool)
   )
   > Clear and Invalidate Timer by Handle(Handle=Get HideOverheatDelayTimerHandle())
   n25 = GetComponentByClass(
    self=@n18
    ComponentClass=&B_ThermalLanceWeaponTargetingComponent_C
   )
   Is Valid(InputObject=@n25)
   > Bind Event to On Targeted Actor Type Changed(self:B Ability Targeting Feedback Component Parent Object Reference=@n25, Delegate=Create Event()) {is=title K2Node_AddDelegate}
   > HandleTargetedActorTypeChanged!(
    execute
    self
    ActorType:AbilityTargetingTargetActorTypeEnum Enum=GetLatestTargetingActorType!(self:B Ability Targeting Feedback Component Parent Object Reference=@n25) -> (Type:AbilityTargetingTargetActorTypeEnum Enum=NewEnumerator0)
   ) -> (then) {is=title K2Node_CallFunction}
   > n26 = Bind Event to On Overheat Value Changed(self:Fort Weapon Ranged Object Reference=@n17, Delegate=Create Event()) {is=title K2Node_AddDelegate}
   HandleValidTargetInRangeChanged!() -> (OutputDelegate, then, bValid:bool=~) {is=title K2Node_CustomEvent}
   Reroute Node(InputPin:Fort Weapon Ranged Object Reference=@n17) -> (OutputPin:Fort Weapon Ranged Object Reference)
  ]}
 PlayActorInRange {nodes [
   n1 = PlayActorInRange!() -> (then, IsActorInRange?:bool=~) {is=title K2Node_FunctionEntry}
   > n2 = Delta Gate Wildcard!(execute:Exec, Variable:"Boolean (by ref)"=Get IsActorInRange() -> (IsActorInRange:bool), New Value:bool=@n1) -> (Success:Exec, Updated Variable:"Boolean (by ref)", Previous Variable:bool=~) {is=title K2Node_MacroInstance}
   > n3 = Branch(Condition=@n2)
   n4 = Get M_CenterReticle() -> (M_CenterReticle:Material Instance Dynamic Object Reference)
   > SetScalarParameterValue(self=@n4, ParameterName=OuterShapeOpacity, Value=0.85)
   n5 = Get Anim_Env_in_Range() -> (Anim_Env_in_Range:Widget Animation Object Reference)
   > PlayAnimationForward(InAnimation=@n5)
   SetScalarParameterValue(execute=@n3.else, self=@n4, ParameterName=OuterShapeOpacity, Value=0.375)
   > PlayAnimationReverse(InAnimation=@n5)
  ]}
 PlayEnemyInRange {nodes [
   n1 = PlayEnemyInRange!() -> (then, IsEnemyInRange:bool=~) {is=title K2Node_FunctionEntry}
   > n2 = Delta Gate Wildcard(Variable=Get IsEnemyInRange() -> (IsEnemyInRange:bool), New Value=@n1) -> (Previous Variable=~)
   > n3 = Branch(Condition=@n2)
   n4 = Get Anim_Enemy_in_Range() -> (Anim_Enemy_in_Range:Widget Animation Object Reference)
   > PlayAnimationForward(InAnimation=@n4)
   PlayAnimationReverse(execute=@n3.else, InAnimation=@n4)
  ]}
 SetupOverheatMeter {nodes [
   SetupOverheatMeter() {is=title K2Node_FunctionEntry}
   > n1 = GetDynamicMaterial(self=Get Image_OverheatMeter() -> (Image_OverheatMeter:Image Object Reference))
   > n2 = Set DynamicMaterialOverheat(DynamicMaterialOverheat:Material Instance Dynamic Object Reference=@n1) -> (Output_Get:Material Instance Dynamic Object Reference)
   > Is Valid(InputObject=@n2)
   > n3 = CalculatePulseProperties!(execute, self) -> (then, Count:int=0, FinalPulseValue:double) {is=title K2Node_CallFunction}
   n4 = int - int(A=@n3)
   > SetScalarParameterValue(self=@n2, ParameterName=PulseNumber, Value="To Float (Integer)"(InInt="Max (Integer)"(A=@n4, B=0)))
   > SetScalarParameterValue(self=@n2, ParameterName=FinalPulseTime, Value=@n3.FinalPulseValue)
  ]}
]
variables [
 bOverheatHidden {type=boolean, default=true}
 bWasOverheated {type=boolean}
 CurrentOverheatValue {type=number}
 DynamicMaterialOverheat {type=MaterialInstanceDynamic}
 HideOverheatDelayTimerHandle {type=TimerHandle}
 Hotfix_PulseFireOverheatCadence {
  type=ScalableFloat
  default {
   registryType=None
   value=1
   curve {curveTable=/TidalCranePrototype/Gameplay/ChargedEnergySniper/ChargedEnergySniperGameData, rowName=ThermalLance.Pulse.Fire.Overheat.Cadence}
  }
 }
 IsActorInRange {type=boolean}
 IsEnemyInRange {type=boolean}
 M_CenterReticle {type=MaterialInstanceDynamic}
 TimeToHideOverheat {type=number, default=1}
 Weapon {type=FortWeaponRanged, default=null}
]
widgets {
 info {
  inheritedWidgetCount=1
  parentClass=/Script/FortniteUI.FortWeaponReticleExtensionWidgetBase
  rootWidgetClass=&Overlay
  widgetCount=4
 }
 widgets [
  Overlay_24 {widgetClassPath=&Overlay, widget=Overlay_24}
  Image_ReticleCenter {
   widgetClassPath=&Image
   bIsVariable=true
   parent=Overlay_24
   widget=Image_ReticleCenter
   properties {brush {
     bIsDynamicallyLoaded=false
     drawAs=Image
     imageType=FullColor
     mirroring=NoMirror
     resourceName=None
     resourceObject=/ThermalLanceGameplay/UI/Materials/M_UI_ThermalLance_Reticle
     tiling=NoTile
     imageSize=<128 128>
     margin=<ltrb 0 0 0 0>
     outlineSettings {
      bUseBrushTransparency=false
      roundingType=HalfHeightRadius
      width=0
      color {colorUseRule=UseColor_Specified, specifiedColor=<rgba 0 0 0 0>}
      cornerRadii=<0 0 0 0>
     }
     tintColor {colorUseRule=UseColor_Specified, specifiedColor=<rgba 1 1 1 1>}
     uVRegion {bIsValid=false, max=<0 0>, min=<0 0>}
    }}
   slot {class=OverlaySlot, properties {horizontalAlignment=HAlign_Center, verticalAlignment=VAlign_Center}}
  }
  Image_OverheatMeter {
   widgetClassPath=&Image
   bIsVariable=true
   parent=Overlay_24
   widget=Image_OverheatMeter
   properties {
    renderOpacity=0
    brush {
     bIsDynamicallyLoaded=false
     drawAs=Image
     imageType=FullColor
     mirroring=NoMirror
     resourceName=None
     resourceObject=/ThermalLanceGameplay/UI/Materials/M_UI_ThermalLance_Overheat
     tiling=NoTile
     imageSize=<162 155.52>
     margin=<ltrb 0 0 0 0>
     outlineSettings {
      bUseBrushTransparency=false
      roundingType=HalfHeightRadius
      width=0
      color {colorUseRule=UseColor_Specified, specifiedColor=<rgba 0 0 0 0>}
      cornerRadii=<0 0 0 0>
     }
     tintColor {colorUseRule=UseColor_Specified, specifiedColor=<rgba 1 1 1 1>}
     uVRegion {bIsValid=false, max=<0 0>, min=<0 0>}
    }
    renderTransform {angle=-40, scale=<1.25 1.25>, shear=<0 0>, translation=<55 0>}
   }
   slot {class=OverlaySlot}
  }
  TutorialHighlight {widgetClassPath=&TutorialHighlight}
 ]
}
